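# Animation Blueprint for the M_MED_Clash backpack skeleton, named as the gameplay version of the pack.
# Its AnimGraph feeds a single Sequence Player straight to the output pose, playing the Clash male Stage03 backpack idle with looping off.
# The EventGraph holds only unconnected update-event and pawn-owner nodes, and the asset declares no variables.
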AnimBlueprint /Game/Accessories/FORT_Backpacks/Backpack_M_MED_Clash/Meshes/M_MED_Clash_Pack_Gameplay_AnimBP
target_skeleton=&/M_MED_Clash_Pack_Skeleton
graphs [
 AnimGraph {nodes [Output Pose(Result=Sequence Player 'Clash_Male_Stage03_Idle_BACKPACK'() {bLoopAnimation=false, sequence=/Game/Accessories/FORT_Backpacks/Backpack_M_MED_Clash/Animation/Clash_Male_Stage03_Idle_BACKPACK}) {layerGroup=None}]}
 EventGraph {nodes [
   TryGetPawnOwner()
   Event BlueprintUpdateAnimation()
  ]}
]
variables []
